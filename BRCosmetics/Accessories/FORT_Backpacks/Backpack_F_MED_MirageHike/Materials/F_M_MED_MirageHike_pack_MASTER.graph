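# Master character material for the MirageHike backpack cosmetic, using material attributes with per-expression shading models on skeletal meshes.
# Shared PreFX output passes through optional thin-film and metal LUT stages, masked via named reroutes.
# Roughness is remapped between min/max parameters within the thin-film channel mask.
# The result feeds shared character PostFX, then cinematic PostFX.
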Material /BRCosmetics/Accessories/FORT_Backpacks/Backpack_F_MED_MirageHike/Materials/F_M_MED_MirageHike_pack_MASTER
graph {
 properties {
  bAutomaticallySetUsageInEditor=false
  bUseMaterialAttributes=true
  bUsedWithMeshDeformer=true
  bUsedWithMorphTargets=true
  bUsedWithSkeletalMesh=true
  pixelDepthOffsetMode=PDOM_Legacy
  refractionMethod=RM_IndexOfRefraction
  shadingModel=MSM_FromMaterialExpression
  subsurfaceProfile=&SS_HeroSkin_02
  usedShadingModels=Default Lit | Subsurface | Clear Coat | Subsurface Profile | Cloth | Eye
 }
 nodes [
  n1 = MaterialFunctionCall() {materialFunction=&BaseFN_CharacterFunction_PreFX}
  n2 = NamedRerouteUsage()
  n3 = StaticComponentMaskParameter(None=NamedRerouteUsage()) {defaultR=true, group=Skin FX, parameterName=ThinFilm_Channel}
  n4 = StaticSwitchParameter(
   True=MaterialFunctionCall!(Alpha=@n3, Material Attributes=@n1.Result) {materialFunction=/Game/Weapons/WeaponSkins/MaterialFunction/ThinFilmFunction}.Result
   False=@n1.Result
  ) {defaultValue=true, group=Skin FX, parameterName=Use Thin Film}
  n5 = GetMaterialAttributes(None=StaticSwitchParameter(
    True=MaterialFunctionCall(Mask=@n2, Materials=@n4) {materialFunction=&MF_MetalsLUT}.Result
    False=@n4
   ) {defaultValue=true, group=Skin FX, parameterName=UseMetalLUT}) {attributeGetTypes [#1]}
  n6 = LinearInterpolate(
   A=ScalarParameter() {defaultValue=0.35, parameterName=Roughness min}
   B=ScalarParameter() {defaultValue=0.5, parameterName=Roughness max, sliderMin=1}
   Alpha=@n5.Roughness
  )
  n7 = MaterialFunctionCall(
   AO=@n1.AO
   Distance Blend=@n1.AO
   Fresnel=@n1.AO
   Gmap=@n1.AO
   Pre Skinned Local Position=@n1.AO
   Material=SetMaterialAttributes!(MaterialAttributes=@n5.MaterialAttributes, Roughness=LinearInterpolate(A=@n5.Roughness, B=@n6, Alpha=@n3)) {attributeSetTypes [#1]}
  ) {materialFunction=&BaseFN_CharacterFunction_PostFX}
  MaterialFunctionCall(In=@n7.Result) {materialFunction=&CINE_MF_CharPostFX}
  MaterialFunctionCall(
   PreskinnedPosition=@n1.Pre skinned Local Position
   Gmap=@n1.Pre skinned Local Position
   MaterialAttributes=@n1.Pre skinned Local Position
   AO=@n1.Pre skinned Local Position
  ) {materialFunction=&CharacterSkinFX}
  n8 = TextureObjectParameter() {group=Skin FX, parameterName=SkinFX_Mask}
  NamedRerouteDeclaration(None=TextureSample(UVs=TextureCoordinate(), Tex=NamedRerouteDeclaration(None=@n8) {name=SkinFX_Mask, nodeColor=<rgba 0.9842 0 1 1>}).RGBA) {name=SkinFX_Mask_Sampled, nodeColor=<rgba 0.7475 0 1 1>}
 ]
}
